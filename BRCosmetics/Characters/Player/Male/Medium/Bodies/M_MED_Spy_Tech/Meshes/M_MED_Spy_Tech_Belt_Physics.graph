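# Physics asset for the belt mesh of the male medium Spy Tech body cosmetic, simulating dangling left-side belt pieces.
# Capsules on the pelvis, lower spine and both thighs approximate the body volume around the belt.
# Front, back and side belt bones each hang from a tiny MSTR anchor capsule with tight swing and twist limits.
# Those anchors attach to the pelvis with free angular motion, as do the spine links, while thighs keep limited joints.
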PhysicsAsset /BRCosmetics/Characters/Player/Male/Medium/Bodies/M_MED_Spy_Tech/Meshes/M_MED_Spy_Tech_Belt_Physics
bodies {
 dyn_belt_01_bk_l [{shapeType=Capsule, length=12.2572, radius=3.8659, center=<9.428 0.0003 0>, rotation=<rot 90 -0.0033 -0.0032>}]
 dyn_belt_01_bk_l_MSTR [{shapeType=Capsule, length=0.01, radius=0.01}]
 dyn_belt_01_fr_l [{shapeType=Capsule, length=21.0072, radius=4.6159, center=<12.3117 0.0003 0>, rotation=<rot 90 -0.0197 -0.0197>}]
 dyn_belt_01_fr_l_MSTR [{shapeType=Capsule, length=0.01, radius=0.01}]
 dyn_belt_01_l [{shapeType=Capsule, length=10.7572, radius=5.3659, center=<22.6456 0.5536 0.0001>, rotation=<rot -0.0002 -0.0001 0>}]
 dyn_belt_01_l_MSTR [{shapeType=Capsule, length=0.01, radius=0.01}]
 pelvis [{shapeType=Capsule, length=8.8415, radius=8.8942}]
 spine_01 [{shapeType=Capsule, length=8.8415, radius=8.8942}]
 spine_02 [{shapeType=Capsule, length=4.5915, radius=8.8942, center=<-0.0001 -2.0435 0>}]
 thigh_l [{shapeType=Capsule, length=21.0915, radius=7.8942, center=<-16.3308 0.0004 0>, rotation=<rot -90 -0.0033 0.0033>}]
 thigh_r [{shapeType=Capsule, length=21.0915, radius=7.8942, center=<16.3308 -0.0004 0>, rotation=<rot -90 0.0396 179.9604>}]
}
constraints [
 {bone1Name=thigh_r, bone2Name=pelvis, swing1LimitDegrees=45, swing2LimitDegrees=45, twistLimitDegrees=45}
 {
  bone1Name=dyn_belt_01_fr_l_MSTR
  bone2Name=pelvis
  swing1LimitDegrees=45
  swing1Motion=Free
  swing2LimitDegrees=45
  swing2Motion=Free
  twistLimitDegrees=45
  twistMotion=Free
 }
 {
  bone1Name=dyn_belt_01_bk_l_MSTR
  bone2Name=pelvis
  swing1LimitDegrees=45
  swing1Motion=Free
  swing2LimitDegrees=45
  swing2Motion=Free
  twistLimitDegrees=45
  twistMotion=Free
 }
 {
  bone1Name=dyn_belt_01_l_MSTR
  bone2Name=pelvis
  swing1LimitDegrees=45
  swing1Motion=Free
  swing2LimitDegrees=45
  swing2Motion=Free
  twistLimitDegrees=45
  twistMotion=Free
 }
 {bone1Name=dyn_belt_01_l, bone2Name=dyn_belt_01_l_MSTR, swing1LimitDegrees=30, swing2LimitDegrees=5, twistLimitDegrees=5}
 {bone1Name=dyn_belt_01_bk_l, bone2Name=dyn_belt_01_bk_l_MSTR, swing1LimitDegrees=15, swing2LimitDegrees=5, twistLimitDegrees=5}
 {bone1Name=dyn_belt_01_fr_l, bone2Name=dyn_belt_01_fr_l_MSTR, swing1LimitDegrees=30, swing2LimitDegrees=30, twistLimitDegrees=5}
 {
  bone1Name=spine_01
  bone2Name=pelvis
  swing1LimitDegrees=45
  swing1Motion=Free
  swing2LimitDegrees=45
  swing2Motion=Free
  twistLimitDegrees=45
  twistMotion=Free
 }
 {
  bone1Name=spine_02
  bone2Name=spine_01
  swing1LimitDegrees=45
  swing1Motion=Free
  swing2LimitDegrees=45
  swing2Motion=Free
  twistLimitDegrees=45
  twistMotion=Free
 }
 {bone1Name=thigh_l, bone2Name=pelvis, swing1LimitDegrees=45, swing2LimitDegrees=45, twistLimitDegrees=45}
]
