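# Physics asset for the medium male Spy Tech player body, set up for secondary motion on the jacket.
# Capsules cover pelvis and lower spine; tiny MSTR anchor capsules hang off spine_02 for jacket panels and pelvis for the belt, which has only anchors.
# Box bodies chain from jacket anchors with limited swing, twist locked on top and mid panels.
# Free-angular cross-links join lower back panels and, on the left only, front to mid.
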PhysicsAsset /BRCosmetics/Characters/Player/Male/Medium/Bodies/M_MED_Spy_Tech/Meshes/M_MED_Spy_Tech_Physics
bodies {
 dyn_belt_01_bk_l_MSTR [{shapeType=Capsule, length=0.01, radius=0.01}]
 dyn_belt_01_fr_l_MSTR [{shapeType=Capsule, length=0.01, radius=0.01}]
 dyn_belt_01_l_MSTR [{shapeType=Capsule, length=0.01, radius=0.01}]
 dyn_jacket_01_bk [{shapeType=Box, extentX=5.8489, extentY=16.6193, extentZ=2.2842, center=<2.8739 0 -0.0002>}]
 dyn_jacket_01_bk_MSTR [{shapeType=Capsule, length=0.01, radius=0.01}]
 dyn_jacket_02_bk [{shapeType=Box, extentX=5.8489, extentY=16.6193, extentZ=2.2842, center=<2.8739 0 -0.0002>}]
 dyn_jacket_bk_01_l [{shapeType=Box, extentX=5.8489, extentY=16.6193, extentZ=1.7842, center=<2.8739 0 -0.0002>}]
 dyn_jacket_bk_01_l_MSTR [{shapeType=Capsule, length=0.01, radius=0.01}]
 dyn_jacket_bk_01_r [{shapeType=Box, extentX=5.8489, extentY=16.6193, extentZ=1.7842, center=<2.8739 0 -0.0002>}]
 dyn_jacket_bk_01_r_MSTR [{shapeType=Capsule, length=0.01, radius=0.01}]
 dyn_jacket_bk_02_l [{shapeType=Box, extentX=5.8489, extentY=16.6193, extentZ=1.7842, center=<2.8739 0 -0.0002>}]
 dyn_jacket_bk_02_r [{shapeType=Box, extentX=5.8489, extentY=16.6193, extentZ=1.7842, center=<2.8739 0 -0.0002>}]
 dyn_jacket_fr_01_l [{shapeType=Box, extentX=5.8489, extentY=16.6193, extentZ=1.7842, center=<2.8739 0 -0.0002>}]
 dyn_jacket_fr_01_l_MSTR [{shapeType=Capsule, length=0.01, radius=0.01}]
 dyn_jacket_fr_01_r [{shapeType=Box, extentX=5.8489, extentY=16.6193, extentZ=1.7842, center=<2.8739 0 -0.0002>}]
 dyn_jacket_fr_01_r_MSTR [{shapeType=Capsule, length=0.01, radius=0.01}]
 dyn_jacket_fr_02_l [{shapeType=Box, extentX=5.8489, extentY=16.6193, extentZ=1.7842, center=<2.8739 0 -0.0002>}]
 dyn_jacket_fr_02_r [{shapeType=Box, extentX=5.8489, extentY=16.6193, extentZ=1.7842, center=<2.8739 0 -0.0002>}]
 dyn_jacket_mid_01_l [{shapeType=Box, extentX=7.5854, extentY=2.9298, extentZ=18.266, center=<3.4844 -0.2076 -0.0001>, rotation=<rot 0 176.5904 -86.7888>}]
 dyn_jacket_mid_01_l_MSTR [{shapeType=Capsule, length=0.01, radius=0.01}]
 dyn_jacket_mid_01_r [{shapeType=Box, extentX=7.5854, extentY=2.9298, extentZ=18.266, center=<3.4844 -0.2076 -0.0001>, rotation=<rot 0 176.5904 -86.7888>}]
 dyn_jacket_mid_01_r_MSTR [{shapeType=Capsule, length=0.01, radius=0.01}]
 dyn_jacket_tp_01_l [{shapeType=Box, extentX=3.0854, extentY=1.4298, extentZ=5.766, center=<1.4547 -0.0867 -0.0001>, rotation=<rot 0 176.5904 -86.7888>}]
 dyn_jacket_tp_01_l_MSTR [{shapeType=Capsule, length=0.01, radius=0.01}]
 dyn_jacket_tp_01_r [{shapeType=Box, extentX=3.0854, extentY=1.4298, extentZ=5.766, center=<1.4547 -0.0867 -0.0001>, rotation=<rot 0 176.5904 -86.7888>}]
 dyn_jacket_tp_01_r_MSTR [{shapeType=Capsule, length=0.01, radius=0.01}]
 pelvis [{shapeType=Capsule, length=8.8415, radius=8.8942}]
 spine_01 [{shapeType=Capsule, length=7.8415, radius=9.8942, center=<0 -0.7664 0>}]
 spine_02 [{shapeType=Capsule, length=6.5915, radius=9.3942, center=<-0.0001 -2.0435 0>}]
}
constraints [
 {
  bone1Name=dyn_jacket_tp_01_l_MSTR
  bone2Name=spine_02
  swing1LimitDegrees=45
  swing1Motion=Free
  swing2LimitDegrees=45
  swing2Motion=Free
  twistLimitDegrees=45
  twistMotion=Free
 }
 {
  bone1Name=dyn_jacket_tp_01_r_MSTR
  bone2Name=spine_02
  swing1LimitDegrees=45
  swing1Motion=Free
  swing2LimitDegrees=45
  swing2Motion=Free
  twistLimitDegrees=45
  twistMotion=Free
 }
 {
  bone1Name=dyn_jacket_mid_01_r_MSTR
  bone2Name=spine_02
  swing1LimitDegrees=45
  swing1Motion=Free
  swing2LimitDegrees=45
  swing2Motion=Free
  twistLimitDegrees=45
  twistMotion=Free
 }
 {
  bone1Name=dyn_jacket_mid_01_l_MSTR
  bone2Name=spine_02
  swing1LimitDegrees=45
  swing1Motion=Free
  swing2LimitDegrees=45
  swing2Motion=Free
  twistLimitDegrees=45
  twistMotion=Free
 }
 {
  bone1Name=dyn_jacket_bk_01_l_MSTR
  bone2Name=spine_02
  swing1LimitDegrees=45
  swing1Motion=Free
  swing2LimitDegrees=45
  swing2Motion=Free
  twistLimitDegrees=45
  twistMotion=Free
 }
 {
  bone1Name=dyn_jacket_fr_01_l_MSTR
  bone2Name=spine_02
  swing1LimitDegrees=45
  swing1Motion=Free
  swing2LimitDegrees=45
  swing2Motion=Free
  twistLimitDegrees=45
  twistMotion=Free
 }
 {
  bone1Name=dyn_jacket_01_bk_MSTR
  bone2Name=spine_02
  swing1LimitDegrees=45
  swing1Motion=Free
  swing2LimitDegrees=45
  swing2Motion=Free
  twistLimitDegrees=45
  twistMotion=Free
 }
 {
  bone1Name=dyn_jacket_fr_01_r_MSTR
  bone2Name=spine_02
  swing1LimitDegrees=45
  swing1Motion=Free
  swing2LimitDegrees=45
  swing2Motion=Free
  twistLimitDegrees=45
  twistMotion=Free
 }
 {
  bone1Name=dyn_jacket_bk_01_r_MSTR
  bone2Name=spine_02
  swing1LimitDegrees=45
  swing1Motion=Free
  swing2LimitDegrees=45
  swing2Motion=Free
  twistLimitDegrees=45
  twistMotion=Free
 }
 {
  bone1Name=dyn_belt_01_fr_l_MSTR
  bone2Name=pelvis
  swing1LimitDegrees=45
  swing1Motion=Free
  swing2LimitDegrees=45
  swing2Motion=Free
  twistLimitDegrees=45
  twistMotion=Free
 }
 {
  bone1Name=dyn_belt_01_bk_l_MSTR
  bone2Name=pelvis
  swing1LimitDegrees=45
  swing1Motion=Free
  swing2LimitDegrees=45
  swing2Motion=Free
  twistLimitDegrees=45
  twistMotion=Free
 }
 {
  bone1Name=dyn_belt_01_l_MSTR
  bone2Name=pelvis
  swing1LimitDegrees=45
  swing1Motion=Free
  swing2LimitDegrees=45
  swing2Motion=Free
  twistLimitDegrees=45
  twistMotion=Free
 }
 {
  bone1Name=dyn_jacket_tp_01_l
  bone2Name=dyn_jacket_tp_01_l_MSTR
  swing1LimitDegrees=1
  swing2LimitDegrees=5
  twistLimitDegrees=1
  twistMotion=Locked
 }
 {
  bone1Name=dyn_jacket_tp_01_r
  bone2Name=dyn_jacket_tp_01_r_MSTR
  swing1LimitDegrees=1
  swing2LimitDegrees=5
  twistLimitDegrees=1
  twistMotion=Locked
 }
 {
  bone1Name=dyn_jacket_mid_01_r
  bone2Name=dyn_jacket_mid_01_r_MSTR
  swing1LimitDegrees=1
  swing2LimitDegrees=10
  twistLimitDegrees=1
  twistMotion=Locked
 }
 {
  bone1Name=dyn_jacket_mid_01_l
  bone2Name=dyn_jacket_mid_01_l_MSTR
  swing1LimitDegrees=1
  swing2LimitDegrees=10
  twistLimitDegrees=1
  twistMotion=Locked
 }
 {bone1Name=dyn_jacket_bk_01_l, bone2Name=dyn_jacket_bk_01_l_MSTR, swing1LimitDegrees=15, swing2LimitDegrees=20, twistLimitDegrees=5}
 {bone1Name=dyn_jacket_bk_02_l, bone2Name=dyn_jacket_bk_01_l, swing1LimitDegrees=15, swing2LimitDegrees=20, twistLimitDegrees=5}
 {bone1Name=dyn_jacket_fr_01_l, bone2Name=dyn_jacket_fr_01_l_MSTR, swing1LimitDegrees=15, swing2LimitDegrees=20, twistLimitDegrees=5}
 {bone1Name=dyn_jacket_fr_02_l, bone2Name=dyn_jacket_fr_01_l, swing1LimitDegrees=15, swing2LimitDegrees=20, twistLimitDegrees=5}
 {bone1Name=dyn_jacket_01_bk, bone2Name=dyn_jacket_01_bk_MSTR, swing1LimitDegrees=15, swing2LimitDegrees=20, twistLimitDegrees=5}
 {bone1Name=dyn_jacket_02_bk, bone2Name=dyn_jacket_01_bk, swing1LimitDegrees=15, swing2LimitDegrees=20, twistLimitDegrees=5}
 {bone1Name=dyn_jacket_fr_01_r, bone2Name=dyn_jacket_fr_01_r_MSTR, swing1LimitDegrees=15, swing2LimitDegrees=20, twistLimitDegrees=5}
 {bone1Name=dyn_jacket_fr_02_r, bone2Name=dyn_jacket_fr_01_r, swing1LimitDegrees=15, swing2LimitDegrees=20, twistLimitDegrees=5}
 {bone1Name=dyn_jacket_bk_01_r, bone2Name=dyn_jacket_bk_01_r_MSTR, swing1LimitDegrees=15, swing2LimitDegrees=20, twistLimitDegrees=5}
 {bone1Name=dyn_jacket_bk_02_r, bone2Name=dyn_jacket_bk_01_r, swing1LimitDegrees=15, swing2LimitDegrees=20, twistLimitDegrees=5}
 {
  bone1Name=dyn_jacket_fr_01_l
  bone2Name=dyn_jacket_mid_01_l
  swing1LimitDegrees=45
  swing1Motion=Free
  swing2LimitDegrees=45
  swing2Motion=Free
  twistLimitDegrees=45
  twistMotion=Free
 }
 {
  bone1Name=spine_01
  bone2Name=pelvis
  swing1LimitDegrees=45
  swing1Motion=Free
  swing2LimitDegrees=45
  swing2Motion=Free
  twistLimitDegrees=45
  twistMotion=Free
 }
 {
  bone1Name=spine_02
  bone2Name=spine_01
  swing1LimitDegrees=45
  swing1Motion=Free
  swing2LimitDegrees=45
  swing2Motion=Free
  twistLimitDegrees=45
  twistMotion=Free
 }
 {
  bone1Name=dyn_jacket_bk_02_r
  bone2Name=dyn_jacket_02_bk
  swing1LimitDegrees=45
  swing1Motion=Free
  swing2LimitDegrees=45
  swing2Motion=Free
  twistLimitDegrees=45
  twistMotion=Free
 }
 {
  bone1Name=dyn_jacket_bk_02_l
  bone2Name=dyn_jacket_02_bk
  swing1LimitDegrees=45
  swing1Motion=Free
  swing2LimitDegrees=45
  swing2Motion=Free
  twistLimitDegrees=45
  twistMotion=Free
 }
]
