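# Animation Blueprint for the M_MED_Spy_Tech player body that follows the attached parent mesh's pose and adds secondary motion to belt and jacket.
# Two pelvis-space RigidBody nodes simulate separate physics assets, with additive bone corrections before and after.
# Crouching toggles belt and left-coat offsets, while being carried in DBNO adjusts the jacket back.
# The update event derives the right-coat correction alpha from 2D speed while crouched, otherwise zero.
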AnimBlueprint /BRCosmetics/Characters/Player/Male/Medium/Bodies/M_MED_Spy_Tech/Meshes/M_MED_Spy_Tech_AnimBP
target_skeleton=&/M_MED_Spy_Tech_Skeleton
graphs [
 AnimGraph {nodes [
   n1 = Get bIsCrouching()
   n2 = "Transform (Modify) Bone - Bone: dyn_belt_01_fr_l_MSTR"!(
    ComponentPose=Local To Component(LocalPose=Copy Pose From Mesh() {bUseAttachedParent=true})
    bAlphaBoolEnabled:bool=@n1
    Translation="0,0,0"
    Rotation="-15.000,-30.000,0.000"
    Scale="1,1,1"
   ) -> (Pose) {
    alphaInputType=Bool
    rotationMode=BMM_Additive
    rotationSpace=BCS_BoneSpace
    translationMode=BMM_Additive
    translationSpace=BCS_BoneSpace
    alphaBoolBlend {blendOption=Linear}
    boneToModify {boneName=dyn_belt_01_fr_l_MSTR}
    rotation {pitch=-15, yaw=-30}
   }
   n3 = "Transform (Modify) Bone - Bone: dyn_belt_01_bk_l_MSTR"!(ComponentPose=@n2, bAlphaBoolEnabled:bool=@n1, Translation="0,0,0", Rotation="0,-20,0", Scale="1,1,1") -> (Pose) {
    alphaInputType=Bool
    rotationMode=BMM_Additive
    rotationSpace=BCS_BoneSpace
    translationMode=BMM_Additive
    translationSpace=BCS_BoneSpace
    alphaBoolBlend {blendOption=Linear}
    boneToModify {boneName=dyn_belt_01_bk_l_MSTR}
    rotation {yaw=-20}
   }
   n4 = "Transform (Modify) Bone - Bone: dyn_belt_01_l_MSTR"!(ComponentPose=@n3, bAlphaBoolEnabled:bool=@n1, Translation="0,0,0", Rotation="0,-20,0", Scale="1,1,1") -> (Pose) {
    alphaInputType=Bool
    rotationMode=BMM_Additive
    rotationSpace=BCS_BoneSpace
    translationMode=BMM_Additive
    translationSpace=BCS_BoneSpace
    alphaBoolBlend {blendOption=Linear}
    boneToModify {boneName=dyn_belt_01_l_MSTR}
    rotation {yaw=-20}
   }
   n5 = RigidBody!(ComponentPose=@n4, Alpha, OverrideWorldGravity:Vector="0,0,-980", ExternalForce) -> (Pose) {
    bClampLinearTranslationLimitToRefPose=true
    bOverrideWorldGravity=true
    overridePhysicsAsset=&/M_MED_Spy_Tech_Physics
    simulationSpace=BaseBoneSpace
    alphaBoolBlend {blendOption=Linear}
    baseBoneRef {boneName=pelvis}
    componentAppliedLinearAccClamp=<3000 3000 3000>
    componentLinearAccScale=<0.3 0.3 0.3>
    componentLinearVelScale=<0.3 0.3 0.5>
    overrideWorldGravity {z=-980}
   }
   n6 = RigidBody(ComponentPose=@n5, OverrideWorldGravity="0,0,-980") {
    bClampLinearTranslationLimitToRefPose=true
    bOverrideWorldGravity=true
    overridePhysicsAsset=&/M_MED_Spy_Tech_Belt_Physics
    simulationSpace=BaseBoneSpace
    alphaBoolBlend {blendOption=Linear}
    baseBoneRef {boneName=pelvis}
    componentAppliedLinearAccClamp=<3000 3000 3000>
    componentLinearAccScale=<0.3 0.3 0.3>
    componentLinearVelScale=<0.3 0.3 0.5>
    overrideWorldGravity {z=-980}
   }
   n7 = "Transform (Modify) Bone - Bone: dyn_belt_02_l"!(ComponentPose=@n6, bAlphaBoolEnabled:bool=Get bIsCrouching(), Translation="0,0,0", Rotation="-20.000,30,0.000", Scale="1,1,1") -> (Pose) {
    alphaInputType=Bool
    rotationMode=BMM_Additive
    rotationSpace=BCS_BoneSpace
    translationMode=BMM_Additive
    translationSpace=BCS_BoneSpace
    alphaBoolBlend {blendOption=Linear}
    boneToModify {boneName=dyn_belt_02_l}
    rotation {pitch=-20, yaw=30}
   }
   n8 = "Transform (Modify) Bone - Bone: dyn_jacket_fr_01_r"(ComponentPose=@n7, Alpha=Get scaleRCoatAlpha() -> (scaleRCoatAlpha:double), Translation="-3,0,0", Rotation="0,0,0", Scale="0.9,1,1") {
    rotationMode=BMM_Additive
    rotationSpace=BCS_BoneSpace
    scaleMode=BMM_Replace
    scaleSpace=BCS_BoneSpace
    translationMode=BMM_Additive
    translationSpace=BCS_BoneSpace
    alphaBoolBlend {blendOption=Linear}
    boneToModify {boneName=dyn_jacket_fr_01_r}
    scale {x=0.9}
    translation {x=-3}
   }
   n9 = "Transform (Modify) Bone - Bone: dyn_jacket_fr_01_l"!(ComponentPose=@n8, bAlphaBoolEnabled:bool=Get bIsCrouching(), Translation="-2.5,0,0", Rotation="0,0,0", Scale="0.9,1,1") -> (Pose) {
    alphaInputType=Bool
    rotationMode=BMM_Additive
    rotationSpace=BCS_BoneSpace
    scaleMode=BMM_Replace
    scaleSpace=BCS_BoneSpace
    translationMode=BMM_Additive
    translationSpace=BCS_BoneSpace
    alphaBoolBlend {blendOption=Linear}
    boneToModify {boneName=dyn_jacket_fr_01_l}
    scale {x=0.9}
    translation {x=-2.5}
   }
   n10 = "Transform (Modify) Bone - Bone: dyn_jacket_01_bk"!(ComponentPose=@n9, bAlphaBoolEnabled:bool=Get bIsDBNOCarried(), Translation="0,0,3", Rotation="-10,0,0", Scale="1,1,1") -> (Pose) {
    alphaInputType=Bool
    rotationMode=BMM_Additive
    rotationSpace=BCS_BoneSpace
    translationMode=BMM_Additive
    translationSpace=BCS_BoneSpace
    alphaBoolBlend {blendOption=Linear}
    boneToModify {boneName=dyn_jacket_01_bk}
    rotation {pitch=-10}
    translation {z=3}
   }
   Output Pose(Result=Component To Local(ComponentPose=@n10)) {layerGroup=None}
  ]}
 EventGraph {nodes [
   Event BlueprintUpdateAnimation()
   n1 = Get PawnSpeed2D()
   > Set scaleRCoatAlpha(scaleRCoatAlpha:double=SelectFloat(A=MapRangeClamped(Value=@n1, InRangeA=0, InRangeB=150), B=0, bPickA=Get bIsCrouching())) -> (Output_Get:double)
  ]}
]
variables [scaleRCoatAlpha {type=number}]
