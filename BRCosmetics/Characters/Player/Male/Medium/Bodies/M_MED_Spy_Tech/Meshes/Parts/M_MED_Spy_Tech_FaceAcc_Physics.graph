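# Physics asset for the M_MED_Spy_Tech face accessory of the medium male player, defining bodies for dynamic hair bones against the upper body.
# Capsules cover head, neck, upper spine and clavicles, while tiny MSTR capsules anchor hair chains and a childless hat root to the head.
# Hair segments have narrow swing and twist limits, tighter at the roots of the two-segment back chains.
# Anchor and body constraints leave angular motion free.
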PhysicsAsset /BRCosmetics/Characters/Player/Male/Medium/Bodies/M_MED_Spy_Tech/Meshes/Parts/M_MED_Spy_Tech_FaceAcc_Physics
bodies {
 clavicle_l [{shapeType=Capsule, length=9.25, radius=8.5, center=<2.5641 -1.0919 -14.4589>, rotation=<rot -64.9921 -179.9998 -180>}]
 clavicle_r [{shapeType=Capsule, length=9.25, radius=8.5, center=<-16.9341 0.3655 13.4927>, rotation=<rot -64.993 -179.9998 0>}]
 dyn_hair_01_bk [{shapeType=Capsule, length=4.5576, radius=1.9556, center=<2.188 -0.0001 0>, rotation=<rot -90 -178.7341 178.7342>}]
 dyn_hair_01_bk_MSTR [{shapeType=Capsule, length=0.01, radius=0.01}]
 dyn_hair_01_bk_l [{shapeType=Capsule, length=4.5576, radius=1.9556, center=<2.188 -0.0001 0>, rotation=<rot -90 -178.7341 178.7342>}]
 dyn_hair_01_bk_l_MSTR [{shapeType=Capsule, length=0.01, radius=0.01}]
 dyn_hair_01_bk_r [{shapeType=Capsule, length=4.5576, radius=1.9556, center=<2.188 -0.0001 0>, rotation=<rot -90 -178.7341 178.7342>}]
 dyn_hair_01_bk_r_MSTR [{shapeType=Capsule, length=0.01, radius=0.01}]
 dyn_hair_01_l [{shapeType=Capsule, length=4.5576, radius=1.9556, center=<2.188 -0.0001 0>, rotation=<rot -90 -178.7341 178.7342>}]
 dyn_hair_01_l_MSTR [{shapeType=Capsule, length=0.01, radius=0.01}]
 dyn_hair_01_r [{shapeType=Capsule, length=4.5576, radius=1.9556, center=<2.188 -0.0001 0>, rotation=<rot -90 -178.7341 178.7342>}]
 dyn_hair_01_r_MSTR [{shapeType=Capsule, length=0.01, radius=0.01}]
 dyn_hair_02_bk [{shapeType=Capsule, length=4.5576, radius=1.9556, center=<2.188 -0.0001 0>, rotation=<rot -90 -178.7341 178.7342>}]
 dyn_hair_02_bk_l [{shapeType=Capsule, length=4.5576, radius=1.9556, center=<2.188 -0.0001 0>, rotation=<rot -90 -178.7341 178.7342>}]
 dyn_hair_02_bk_r [{shapeType=Capsule, length=4.5576, radius=1.9556, center=<2.188 -0.0001 0>, rotation=<rot -90 -178.7341 178.7342>}]
 dyn_hair_fr_01_l [{shapeType=Capsule, length=4.5576, radius=1.9556, center=<2.188 -0.0001 0>, rotation=<rot -90 -178.7341 178.7342>}]
 dyn_hair_fr_01_l_MSTR [{shapeType=Capsule, length=0.01, radius=0.01}]
 dyn_hair_fr_01_r [{shapeType=Capsule, length=4.5576, radius=1.9556, center=<2.188 -0.0001 0>, rotation=<rot -90 -178.7341 178.7342>}]
 dyn_hair_fr_01_r_MSTR [{shapeType=Capsule, length=0.01, radius=0.01}]
 dyn_hat_01_MSTR [{shapeType=Capsule, length=0.01, radius=0.01}]
 head [
  {shapeType=Capsule, length=11.5, radius=5, center=<2.85 0.4895 0>, rotation=<rot -90 0 5.0002>}
  {shapeType=Capsule, length=5.75, radius=5.5, center=<1.2188 7.3671 -0.0001>, rotation=<rot 90 -0.0198 -0.0198>}
 ]
 neck_01 [{shapeType=Capsule, length=5.25, radius=4.5, center=<3.7805 -0.0001 0>, rotation=<rot -90 -7.5524 7.5525>}]
 spine_05 [{shapeType=Capsule, length=17.505, radius=8.755, center=<3.1326 -1.4854 0>, rotation=<rot 0 -90 0>}]
}
constraints [
 {bone1Name=dyn_hair_01_bk_l, bone2Name=dyn_hair_01_bk_l_MSTR, swing1LimitDegrees=10, swing2LimitDegrees=10, twistLimitDegrees=5}
 {bone1Name=dyn_hair_02_bk_l, bone2Name=dyn_hair_01_bk_l, swing1LimitDegrees=20, swing2LimitDegrees=20, twistLimitDegrees=5}
 {bone1Name=dyn_hair_01_l, bone2Name=dyn_hair_01_l_MSTR, swing1LimitDegrees=20, swing2LimitDegrees=20, twistLimitDegrees=5}
 {bone1Name=dyn_hair_fr_01_l, bone2Name=dyn_hair_fr_01_l_MSTR, swing1LimitDegrees=20, swing2LimitDegrees=20, twistLimitDegrees=5}
 {bone1Name=dyn_hair_01_bk_r, bone2Name=dyn_hair_01_bk_r_MSTR, swing1LimitDegrees=10, swing2LimitDegrees=10, twistLimitDegrees=5}
 {bone1Name=dyn_hair_02_bk_r, bone2Name=dyn_hair_01_bk_r, swing1LimitDegrees=20, swing2LimitDegrees=20, twistLimitDegrees=5}
 {bone1Name=dyn_hair_fr_01_r, bone2Name=dyn_hair_fr_01_r_MSTR, swing1LimitDegrees=20, swing2LimitDegrees=20, twistLimitDegrees=5}
 {bone1Name=dyn_hair_01_r, bone2Name=dyn_hair_01_r_MSTR, swing1LimitDegrees=20, swing2LimitDegrees=20, twistLimitDegrees=5}
 {bone1Name=dyn_hair_01_bk, bone2Name=dyn_hair_01_bk_MSTR, swing1LimitDegrees=10, swing2LimitDegrees=10, twistLimitDegrees=5}
 {bone1Name=dyn_hair_02_bk, bone2Name=dyn_hair_01_bk, swing1LimitDegrees=20, swing2LimitDegrees=20, twistLimitDegrees=5}
 {
  bone1Name=dyn_hair_01_bk_l_MSTR
  bone2Name=head
  swing1LimitDegrees=45
  swing1Motion=Free
  swing2LimitDegrees=45
  swing2Motion=Free
  twistLimitDegrees=45
  twistMotion=Free
 }
 {
  bone1Name=dyn_hair_01_l_MSTR
  bone2Name=head
  swing1LimitDegrees=45
  swing1Motion=Free
  swing2LimitDegrees=45
  swing2Motion=Free
  twistLimitDegrees=45
  twistMotion=Free
 }
 {
  bone1Name=dyn_hair_fr_01_l_MSTR
  bone2Name=head
  swing1LimitDegrees=45
  swing1Motion=Free
  swing2LimitDegrees=45
  swing2Motion=Free
  twistLimitDegrees=45
  twistMotion=Free
 }
 {
  bone1Name=dyn_hair_01_bk_r_MSTR
  bone2Name=head
  swing1LimitDegrees=45
  swing1Motion=Free
  swing2LimitDegrees=45
  swing2Motion=Free
  twistLimitDegrees=45
  twistMotion=Free
 }
 {
  bone1Name=dyn_hair_fr_01_r_MSTR
  bone2Name=head
  swing1LimitDegrees=45
  swing1Motion=Free
  swing2LimitDegrees=45
  swing2Motion=Free
  twistLimitDegrees=45
  twistMotion=Free
 }
 {
  bone1Name=dyn_hair_01_r_MSTR
  bone2Name=head
  swing1LimitDegrees=45
  swing1Motion=Free
  swing2LimitDegrees=45
  swing2Motion=Free
  twistLimitDegrees=45
  twistMotion=Free
 }
 {
  bone1Name=dyn_hair_01_bk_MSTR
  bone2Name=head
  swing1LimitDegrees=45
  swing1Motion=Free
  swing2LimitDegrees=45
  swing2Motion=Free
  twistLimitDegrees=45
  twistMotion=Free
 }
 {
  bone1Name=dyn_hat_01_MSTR
  bone2Name=head
  swing1LimitDegrees=45
  swing1Motion=Free
  swing2LimitDegrees=45
  swing2Motion=Free
  twistLimitDegrees=45
  twistMotion=Free
 }
 {
  bone1Name=head
  bone2Name=neck_01
  swing1LimitDegrees=45
  swing1Motion=Free
  swing2LimitDegrees=45
  swing2Motion=Free
  twistLimitDegrees=45
  twistMotion=Free
 }
 {
  bone1Name=clavicle_l
  bone2Name=spine_05
  swing1LimitDegrees=45
  swing1Motion=Free
  swing2LimitDegrees=45
  swing2Motion=Free
  twistLimitDegrees=45
  twistMotion=Free
 }
 {
  bone1Name=clavicle_r
  bone2Name=spine_05
  swing1LimitDegrees=45
  swing1Motion=Free
  swing2LimitDegrees=45
  swing2Motion=Free
  twistLimitDegrees=45
  twistMotion=Free
 }
 {
  bone1Name=neck_01
  bone2Name=spine_05
  swing1LimitDegrees=45
  swing1Motion=Free
  swing2LimitDegrees=45
  swing2Motion=Free
  twistLimitDegrees=45
  twistMotion=Free
 }
]
