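# Post-mesh-copy AnimBlueprint for the Spy Tech male face accessory, adding procedural hat and hair motion to the pose copied from the attached parent.
# Each update converts head twist relative to spine_05 into hatAlphaRef, which blends an additive pitch on dyn_hat_01.
# Three Look At nodes then point back-hair master bones at their aim bones.
# A RigidBody node finally simulates the accessory physics asset in spine_04 space with overridden gravity.
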AnimBlueprint /BRCosmetics/Characters/Player/Male/Medium/Bodies/M_MED_Spy_Tech/Meshes/Parts/M_MED_Spy_Tech_FaceAcc_AnimBP
target_skeleton=&/M_MED_Spy_Tech_FaceAcc_Skeleton
graphs [
 AnimGraph {nodes [
   n1 = "Transform (Modify) Bone - Bone: dyn_hat_01"(
    ComponentPose=Local To Component(LocalPose=Copy Pose From Mesh() {bUseAttachedParent=true})
    Alpha=Get hatAlphaRef() -> (hatAlphaRef:double)
    Translation="0,0,0"
    Rotation="25.000,0.000,0.000"
    Scale="1,1,1"
   ) {
    rotationMode=BMM_Additive
    rotationSpace=BCS_BoneSpace
    translationMode=BMM_Additive
    translationSpace=BCS_BoneSpace
    alphaBoolBlend {blendOption=Linear}
    boneToModify {boneName=dyn_hat_01}
    rotation {pitch=25}
   }
   n2 = Look At - Bone: dyn_hair_01_bk_MSTR(ComponentPose=@n1, Alpha=1) {
    alphaBoolBlend {blendOption=Linear}
    boneToModify {boneName=dyn_hair_01_bk_MSTR}
    lookAtLocation {x=0}
    lookAtTarget {boneReference {boneName=dyn_hair_03_bk_aim}}
    lookAt_Axis {axis=<1 0>}
    lookUp_Axis {axis {x=0, z=1}}
   }
   n3 = Look At - Bone: dyn_hair_01_bk_r_MSTR(ComponentPose=@n2, Alpha=1) {
    alphaBoolBlend {blendOption=Linear}
    boneToModify {boneName=dyn_hair_01_bk_r_MSTR}
    lookAtLocation {x=0}
    lookAtTarget {boneReference {boneName=dyn_hair_03_bk_r_aim}}
    lookAt_Axis {axis=<1 0>}
    lookUp_Axis {axis {x=0, z=1}}
   }
   n4 = Look At - Bone: dyn_hair_01_bk_l_MSTR(ComponentPose=@n3, Alpha=1) {
    alphaBoolBlend {blendOption=Linear}
    boneToModify {boneName=dyn_hair_01_bk_l_MSTR}
    lookAtLocation {x=0}
    lookAtTarget {boneReference {boneName=dyn_hair_03_bk_l_aim}}
    lookAt_Axis {axis=<1 0>}
    lookUp_Axis {axis {x=0, z=1}}
   }
   n5 = RigidBody!(ComponentPose=@n4, Alpha, OverrideWorldGravity:Vector="0,0,-980", ExternalForce) -> (Pose) {
    bClampLinearTranslationLimitToRefPose=true
    bOverrideWorldGravity=true
    overridePhysicsAsset=&/M_MED_Spy_Tech_FaceAcc_Physics
    simulationSpace=BaseBoneSpace
    alphaBoolBlend {blendOption=Linear}
    baseBoneRef {boneName=spine_04}
    componentAppliedLinearAccClamp=<3000 3000 3000>
    componentLinearAccScale=<0.3 0.3 0.3>
    componentLinearVelScale=<0.3 0.3 0.5>
    overrideWorldGravity {z=-980}
   }
   Output Pose(Result=Component To Local(ComponentPose=@n5)) {layerGroup=None}
  ]}
 EventGraph {nodes [
   Event BlueprintUpdateAnimation()
   > Sequence()
   > Set hatAlphaRef(hatAlphaRef:double=GetAlphaFromReferenceTransformDelta(
     SkeletalMeshComponent=GetOwnerSkeletalMeshComponent()
     BoneName=head
     BaseName=spine_05
     DeltaComponentType=QuaternionTwist
     InputRange=(X=0,Y=30)
     TwistAxis="0,0,1"
    )) -> (Output_Get:double)
   TryGetPawnOwner()
  ]}
]
variables [hatAlphaRef {type=number}]
